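# Material function that provides a bronze metal look, named for the Season 27 superlevel player character styles.
# It outputs material attributes built from constants only: a warm brown-orange base color, full metallic and fairly low roughness for a polished sheen.
# All other attributes stay at defaults, and it takes no inputs or textures.
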MaterialFunction /Game/Characters/Player/Common/MaterialFunctions/Superlevel_S27/MF_S27SL_Bronze
graph {nodes [FunctionOutput(None=MakeMaterialAttributes!(
    BaseColor=Constant3Vector() {constant=<rgba 0.35 0.1867 0.07 1>}
    Metallic=Constant() {r=1}
    Specular
    Roughness=Constant() {r=0.3}
    Anisotropy
    EmissiveColor
    Opacity
    OpacityMask
    Normal
    Tangent
    WorldPositionOffset
    SubsurfaceColor
    ClearCoat
    ClearCoatRoughness
    AmbientOcclusion
    Refraction
    CustomizedUVs_0
    CustomizedUVs_1
    CustomizedUVs_2
    CustomizedUVs_3
    CustomizedUVs_4
    CustomizedUVs_5
    CustomizedUVs_6
    CustomizedUVs_7
    PixelDepthOffset
    ShadingModel
    Displacement
   ))]}
