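# Material function for the Season 27 Superlevel player character styles that outputs a cyan, metallic, low-roughness glow as material attributes.
# Emissive intensity comes from a time-of-day scalar, whose daytime value rises with material quality level.
# A powered normal-to-camera dot product blends that intensity, making camera-facing surfaces glow fully and dimming grazing edges, the inverse of a rim light.
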MaterialFunction /Game/Characters/Player/Common/MaterialFunctions/Superlevel_S27/MF_S27SL_Glow
graph {nodes [
  n1 = Constant3Vector() {constant=<rgba 0 0.8 0.7815 1>}
  n2 = Constant() {r=2}
  n3 = MaterialFunctionCall(
   Value_Morning=Constant() {r=1.4}
   Value_Day=QualitySwitch(Default=@n2, High=Multiply(A=@n2) {constB=1.25}, Epic=Multiply(A=@n2) {constB=1.5})
   Value_Evening=Constant() {r=1.2}
   Value_Night=Constant() {r=0.7}
  ) {materialFunction=&Scalar_Time_of_Day}
  n4 = DotProduct(A=PixelNormalWS(), B=CameraVectorWS())
  n5 = LinearInterpolate(A=Multiply(A=Constant() {r=0.2}, B=@n3.Result), B=@n3.Result, Alpha=Power(Base=Saturate(None=@n4), Exp=Constant() {r=4}))
  FunctionOutput(None=MakeMaterialAttributes!(
    BaseColor=@n1
    Metallic=Constant() {r=1}
    Specular
    Roughness=Constant() {r=0.2}
    Anisotropy
    EmissiveColor=Multiply(A=@n1, B=@n5)
    Opacity
    OpacityMask
    Normal
    Tangent
    WorldPositionOffset
    SubsurfaceColor
    ClearCoat
    ClearCoatRoughness
    AmbientOcclusion
    Refraction
    CustomizedUVs_0
    CustomizedUVs_1
    CustomizedUVs_2
    CustomizedUVs_3
    CustomizedUVs_4
    CustomizedUVs_5
    CustomizedUVs_6
    CustomizedUVs_7
    PixelDepthOffset
    ShadingModel
    Displacement
   ))
 ]}
