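# Material function that builds a fully metallic gold surface as Material Attributes, filed under the player character Superlevel S27 functions.
# A saturated camera-to-pixel-normal dot product samples a metal reflectance curve atlas row, tinting gold base color by view angle.
# A quality switch weakens that tint blend on Low and Medium.
# The BaseRoughness input, centered and scaled by an impact parameter, offsets a low roughness that also acts as the minimum.
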MaterialFunction /Game/Characters/Player/Common/MaterialFunctions/Superlevel_S27/MF_S27SL_Gold
graph {nodes [
  n1 = Constant3Vector() {constant=<rgba 1 0.766 0.336 1>}
  n2 = CameraVectorWS()
  n3 = PixelNormalWS()
  n4 = CurveAtlasRowParameter(CurveTime=Saturate(None=DotProduct(A=@n2, B=@n3))) {
   atlas=/Game/Global/Textures/Default/Curve/FN_Character_Atlas
   curve=/Game/Characters/Player/Female/Medium/Bodies/F_MED_Ms_Alpine/Textures/AluminiumreflectanceCurve
   group=06_Metal LUT
   parameterName=Metal LUT Curve
  }
  n5 = Multiply(A=@n1, B=@n4.R)
  n6 = Constant() {r=0.35}
  n7 = QualitySwitch(Default=Constant() {r=0.7}, Low=@n6, Medium=@n6)
  n8 = Constant() {r=0.15}
  n9 = FunctionInput() {inputName=BaseRoughness, inputType=FunctionInput_Scalar}
  n10 = Add(A=@n8, B=Multiply(A=Subtract(A=@n9) {constB=0.5}, B=ScalarParameter() {defaultValue=0.5, parameterName=BaseRoughnessImpact}))
  FunctionOutput(None=MakeMaterialAttributes!(
    BaseColor=LinearInterpolate(A=@n1, B=@n5, Alpha=@n7)
    Metallic=Constant() {r=1}
    Specular
    Roughness=Max(A=@n8, B=@n10)
    Anisotropy
    EmissiveColor
    Opacity
    OpacityMask
    Normal
    Tangent
    WorldPositionOffset
    SubsurfaceColor
    ClearCoat
    ClearCoatRoughness
    AmbientOcclusion
    Refraction
    CustomizedUVs_0
    CustomizedUVs_1
    CustomizedUVs_2
    CustomizedUVs_3
    CustomizedUVs_4
    CustomizedUVs_5
    CustomizedUVs_6
    CustomizedUVs_7
    PixelDepthOffset
    ShadingModel
    Displacement
   ))
 ]}
