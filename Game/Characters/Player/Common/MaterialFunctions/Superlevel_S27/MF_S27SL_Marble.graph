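# Material function that builds the Season 27 Superlevel white marble look for player characters, output as subsurface material attributes.
# A tiled, contrast-boosted marble texture is blended toward a solid color by an anti-seam SphereMask on pre-skinned position, mirrored in X with adjustable offset, scale and skew.
# The result is tinted and darkened by inverted Fresnel (softened on mobile) for BaseColor, and separately drives SubsurfaceColor.
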MaterialFunction /Game/Characters/Player/Common/MaterialFunctions/Superlevel_S27/MF_S27SL_Marble
graph {nodes [
  Composite() {subgraphName=SeamMask}
  n1 = OneMinus(None=Fresnel() {exponent=0.6})
  n2 = ShadingPathSwitch(Default=@n1, Mobile=LinearInterpolate(Alpha=@n1) {constA=0.5}) {desc=Migrated from FeatureLevelSwitch}
  n3 = Multiply(A=TextureCoordinate(), B=ScalarParameter() {defaultValue=2.52, group=SuperLevel, parameterName=MarbleTiling})
  n4 = PreSkinnedPosition()
  n5 = MaterialFunctionCall(Float3=FunctionInput(Preview=VertexInterpolator(VS=@n4).PS) {bUsePreviewValueAsDefault=true, inputName=PreSkinPos}) {materialFunction=&BreakOutFloat3Components}
  n6 = Add(
   A=MaterialFunctionCall(X=Abs(None=@n5.R), Y=@n5.G, Z=@n5.G) {materialFunction=&MakeFloat3}.Result
   B=VectorParameter() {group=SuperLevel, parameterName=AntiSeamPosition, defaultValue=<rgba -13.5 10 -15 1>}.RGB
  )
  n7 = ScalarParameter() {defaultValue=1, group=SuperLevel, parameterName=AntiSeamSize_X}
  n8 = ScalarParameter() {defaultValue=1, group=SuperLevel, parameterName=AntiSeamSize_Y}
  n9 = ComponentMask(None=@n6) {b=true}
  n10 = Divide(A=@n9, B=ScalarParameter() {defaultValue=50, group=SuperLevel, parameterName=AntiSeamSkewExtent}) {constB=50}
  n11 = ScalarParameter() {group=SuperLevel, parameterName=AntiSeamSkew}
  n12 = Add(
   A=MaterialFunctionCall(UVs=ComponentMask(None=@n6) {g=true, r=true}, Texture Scale=AppendVector(A=@n7, B=@n8)) {materialFunction=&ScaleUVsByCenter}.UVs
   B=MaterialFunctionCall(Y=Multiply(A=@n10, B=@n11)) {materialFunction=&MakeFloat2}.Result
  )
  n13 = Multiply(A=@n9, B=ScalarParameter() {defaultValue=1, group=SuperLevel, parameterName=AntiSeamSize_Z})
  n14 = Multiply(
   A=SphereMask(
    A=AppendVector(A=@n12, B=@n13)
    B=Constant()
    Radius=ScalarParameter() {defaultValue=10, group=SuperLevel, parameterName=AntiSeamSize}
    Hardness=ScalarParameter() {defaultValue=0.25, group=SuperLevel, parameterName=AntiSeamHardness}
   ) {attenuationRadius=10, hardnessPercent=50}
   B=ScalarParameter() {defaultValue=1, group=SuperLevel, parameterName=AntiSeamIntensity}
  )
  MaterialExpressionReroute_0 = Reroute(None=@n14)
  n15 = LinearInterpolate(
   A=MaterialFunctionCall!(
    In=TextureSample(UVs=@n3) {texture=&/T_Asteria_Lair_White_Marble_D_HighContrast}.RGB
    Contrast=Constant() {r=0.3}
   ) {materialFunction=/Engine/Functions/Engine_MaterialFunctions01/ImageAdjustment/CheapContrast_RGB}.Result
   B=VectorParameter() {group=SuperLevel, parameterName=AntiSeamColor, defaultValue=<rgba 0.882 0.8946 0.9 1>}.RGB
   Alpha=@MaterialExpressionReroute_0.Output
  )
  n16 = Multiply(
   A=@n15
   B=Multiply(A=Constant3Vector() {constant=<rgba 0.6979 0.6521 0.6325 1>}, B=ScalarParameter() {defaultValue=1.2, parameterName=MarbleMobileBrightness})
  )
  n17 = Saturate(None=Multiply(A=@n15) {constB=1.25})
  n18 = VectorParameter() {parameterName=Subsurface Brightness, defaultValue=<rgba 0.1354 0.1143 0.1019 1>}
  FunctionOutput(None=MakeMaterialAttributes!(
    BaseColor=Multiply(A=@n2, B=@n16)
    Metallic=Constant() {r=0.25}
    Specular
    Roughness=Constant() {r=0.22}
    Anisotropy
    EmissiveColor
    Opacity=Constant()
    OpacityMask
    Normal
    Tangent
    WorldPositionOffset
    SubsurfaceColor=Multiply(A=@n17, B=@n18.RGB)
    ClearCoat
    ClearCoatRoughness
    AmbientOcclusion
    Refraction
    CustomizedUVs_0
    CustomizedUVs_1
    CustomizedUVs_2
    CustomizedUVs_3
    CustomizedUVs_4
    CustomizedUVs_5
    CustomizedUVs_6
    CustomizedUVs_7
    PixelDepthOffset
    ShadingModel=ShadingModel() {shadingModel=MSM_Subsurface}
    Displacement
   ))
  PinBase()
  PinBase!(Output=@n14) {reroutePins [Output {expression=&self:MaterialExpressionReroute_0}]}
 ]}
