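# Material function for Season 27 superlevel player character styles, layering glow, marble, gold and bronze looks onto input material attributes.
# FXMask channels supply gold (R), emissive (G) and marble (B) masks, with bronze derived from squared, smoothstepped R.
# Sub-function layers are blended in a tree, with a power-sharpened glow alpha that UseOriginalEmissive zeroes.
# Marble takes a vertex-interpolated pre-skinned position by default, and selected blended attributes are written back onto the input.
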MaterialFunction /Game/Characters/Player/Common/MaterialFunctions/Superlevel_S27/MF_S27SL
graph {nodes [
  n1 = FunctionInput() {inputName=InMA, inputType=FunctionInput_MaterialAttributes}
  n2 = NamedRerouteUsage()
  n3 = StaticSwitch(
   True=Constant()
   False=Constant() {r=1}
   Value=FunctionInput(Preview=StaticBool()) {bUsePreviewValueAsDefault=true, inputName=UseOriginalEmissive, inputType=FunctionInput_StaticBool, sortPriority=20}
  )
  n4 = BlendMaterialAttributes(
   A=@n1
   B=MaterialFunctionCall() {materialFunction=&/MF_S27SL_Glow}.Result
   Alpha=Power(Base=Multiply(A=@n2, B=@n3)) {constExponent=4}
  )
  n5 = PreSkinnedPosition()
  n6 = MaterialFunctionCall!(PreSkinPos=FunctionInput(Preview=VertexInterpolator(VS=@n5).PS) {bUsePreviewValueAsDefault=true, inputName=PreSkinPos, sortPriority=100}) {materialFunction=&/MF_S27SL_Marble}
  n7 = NamedRerouteUsage()
  n8 = MaterialFunctionCall!(BaseRoughness=GetMaterialAttributes(None=@n1) {attributeGetTypes [#1]}.Roughness) {materialFunction=&/MF_S27SL_Gold}
  n9 = MaterialFunctionCall() {materialFunction=&/MF_S27SL_Bronze}
  n10 = NamedRerouteUsage()
  n11 = GetMaterialAttributes(None=BlendMaterialAttributes(
    A=BlendMaterialAttributes(A=@n4, B=@n6.Result, Alpha=@n7)
    B=BlendMaterialAttributes(A=@n8.Result, B=@n9.Result, Alpha=@n10)
    Alpha=NamedRerouteUsage()
   )) {attributeGetTypes [
    #2
    #3
    #1
    #4
    #5
    #6
    #7
   ]}
  n12 = SetMaterialAttributes!(
   MaterialAttributes=@n1
   BaseColor=@n11.BaseColor
   Metallic=@n11.BaseColor
   Roughness=@n11.BaseColor
   EmissiveColor=@n11.BaseColor
   Opacity=@n11.BaseColor
   SubsurfaceColor=@n11.BaseColor
   ShadingModel=@n11.BaseColor
  ) {attributeSetTypes [
    #2
    #3
    #1
    #4
    #5
    #6
    #7
   ]}
  FunctionOutput(None=@n12) {outputName=OutMA}
  n13 = MaterialFunctionCall(Float3=FunctionInput() {inputName=FXMask, sortPriority=10}) {materialFunction=&BreakOutFloat3Components}
  NamedRerouteDeclaration(None=@n13.R) {name=GoldMask, nodeColor=<rgba 1 0.6854 0 1>}
  NamedRerouteDeclaration(None=SmoothStep(Value=Multiply(A=@n13.R, B=@n13.R)) {constMin=0.5}) {name=BronzeMask, nodeColor=<rgba 0.2726 0.0909 0 1>}
  NamedRerouteDeclaration(None=@n13.G) {name=EmissiveMask, nodeColor=<rgba 0 0.802 1 1>}
  NamedRerouteDeclaration(None=@n13.B) {name=MarbleMask, nodeColor=<rgba 1 0.8962 0.6337 1>}
 ]}
